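# Translucent UI material for a FlatWare quest reward background in Battle Pass S30.
# Procedural gradient, vignette and radial bottom glow layers, plus an overlay texture, are built as named reroutes for a multiply, Soft Light and Color Dodge color stack.
# Static switches toggle the overlay, top-left vignette, static texture and an animatable angled left bar.
# A separate mask takes the minimum of an optional skewed box SDF and two optional linear fades.
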Material /BattlePassS30_Quests/Items/QuestRewards/FlatWare/Materials/M_UI_FlatWare_Simple_BG
graph {
 properties {blendMode=BLEND_Translucent, materialDomain=MD_UI, pixelDepthOffsetMode=PDOM_Legacy, usedShadingModels=Default Lit}
 nodes [
  n1 = ConstantBiasScale(None=MaterialFunctionCall(
    UVs=TextureCoordinate()
    "Rotation Angle (0-1)"=ScalarParameter() {defaultValue=0.5, group=Skew, parameterName=Skew Rotation, sliderMax=1, sortPriority=36}
   ) {materialFunction=&CustomRotator}.Rotated Values) {bias=-0.5, scale=2}
  n2 = LinearInterpolate(
   B=Multiply(
    A=ComponentMask(None=@n1) {r=true}
    B=ScalarParameter() {defaultValue=1, group=Skew, parameterName=Skew Intensity, sliderMax=1, sortPriority=35}
   )
   Alpha=Saturate(None=ScalarParameter() {defaultValue=0.2, group=Skew, parameterName=Skew Amount, sliderMax=1, sortPriority=34})
  ) {constA=1}
  n3 = Multiply(A=ScalarParameter() {group=Skew, parameterName=Skew Position X, sliderMax=1, sliderMin=-1, sortPriority=39}) {constB=-1}
  n4 = ScalarParameter() {group=Skew, parameterName=Skew Position Y, sliderMax=1, sliderMin=-1, sortPriority=40}
  n5 = AppendVector(
   A=ScalarParameter() {defaultValue=0.4, group=Skew, parameterName=Skew Width, sortPriority=37}
   B=ScalarParameter() {defaultValue=0.4, group=Skew, parameterName=Skew Height, sortPriority=38}
  )
  MaterialExpressionSubtract_0 = Subtract(
   A=Add(
    A=MaterialFunctionCall(Texture Scale=@n2) {materialFunction=&ScaleUVsByCenter}.UVs
    B=AppendVector(A=@n3, B=@n4)
   )
   B=OneMinus(None=@n5)
  )
  MaterialExpressionSubtract_1 = Subtract(A=@n5) {constB=0.1}
  n6 = Custom!(p=@MaterialExpressionSubtract_0, b=@MaterialExpressionSubtract_1) {
   code="float2 d = abs(p)-b;\r\nreturn length(max(d,0.0)) + min(max(d.x,d.y),0.0);"
   description=Square SDF
   inputs [
    {
     inputName=p
     input {expression=&self:MaterialExpressionSubtract_0}
    }
    {
     inputName=b
     input {expression=&self:MaterialExpressionSubtract_1}
    }
   ]
  }
  n7 = OneMinus(None=SmoothStep(Max=ScalarParameter() {defaultValue=0.01, group=Skew, parameterName=Skew Feather, sliderMax=1, sortPriority=39}, Value=@n6) {constMax=0.01})
  n8 = Constant() {r=1}
  n9 = MaterialFunctionCall(
   GradientStart=ScalarParameter() {group=Fade, parameterName=Fade Start, sortPriority=34}
   GradientEnd=ScalarParameter() {defaultValue=1, group=Fade, parameterName=Fade End, sortPriority=35}
   RotationAngle=ScalarParameter() {group=Fade, parameterName=Fade Rotation, sliderMax=1, sortPriority=36}
  ) {materialFunction=&MF_UI_LinearGradient}
  n10 = ScalarParameter() {group=Right Fade, parameterName=R Fade Start, sortPriority=34}
  n11 = ScalarParameter() {defaultValue=1, group=Right Fade, parameterName=R Fade End, sortPriority=35}
  n12 = ScalarParameter() {group=Right Fade, parameterName=R Fade Rotation, sliderMax=1, sortPriority=36}
  Min(
   A=Min(
    A=StaticSwitchParameter(True=@n7, False=@n8) {group=Skew, parameterName=Enable Skew, sortPriority=33}
    B=StaticSwitchParameter(True=@n9.Result, False=@n8) {group=Fade, parameterName=Enable Fade, sortPriority=33}
   )
   B=StaticSwitchParameter(
    True=MaterialFunctionCall(GradientStart=@n10, GradientEnd=@n11, RotationAngle=@n12) {materialFunction=&MF_UI_LinearGradient}.Result
    False=@n8
   ) {group=Right Fade, parameterName=Enable R Fade, sortPriority=33}
  )
  n13 = MaterialFunctionCall() {materialFunction=&GetUserInterfaceUV}
  n14 = DotProduct(A=OneMinus(None=@n13.Normalized UV), B=@n13.Normalized UV)
  n15 = Power(Base=Saturate(None=Multiply(A=@n14) {constB=5})) {constExponent=0}
  n16 = ComponentMask(None=@n13.Normalized UV) {g=true}
  n17 = MaterialFunctionCall("Value (0-1)"=@n16) {materialFunction=&MF_UI_EaseCurves}
  NamedRerouteDeclaration(None=Saturate(None=LinearInterpolate(B=@n15, Alpha=@n17.EaseInOutSine) {constA=1})) {name=Vignette, nodeColor=<rgba 0 0 0 1>}
  n18 = MaterialFunctionCall(
   UVs=MaterialFunctionCall() {materialFunction=&GetUserInterfaceUV}.Normalized UV
   Texture Scale=AppendVector(A=Constant() {r=2}, B=Constant() {r=0.5})
  ) {materialFunction=&ScaleUVsByCenter}
  n19 = ScalarParameter() {parameterName=X}
  n20 = ScalarParameter() {defaultValue=-1.1, parameterName=Y}
  n21 = MaterialFunctionCall(UV=Add(
    A=MaterialFunctionCall(UV=@n18.UVs) {materialFunction=&MF_UI_UVScaling}.Scaled To Fill
    B=AppendVector(A=@n19, B=@n20)
   )) {materialFunction=&MF_UI_RadialGradient}
  NamedRerouteDeclaration(None=LinearInterpolate(A=VectorParameter() {parameterName=Glow, defaultValue=<rgba 0.9647 0.2874 0.1221 1>}.RGB, Alpha=OneMinus(None=@n21.Smooth)) {constB=0}) {name=Bottom Glow, nodeColor=<rgba 0 0 0 1>}
  n22 = VectorParameter() {parameterName=LinearGradientColorA, defaultValue=<rgba 1 0.5711 0.4621 1>}
  n23 = VectorParameter() {parameterName=LinearGradientColorB, defaultValue=<rgba 1 0.5089 0 1>}
  n24 = MaterialFunctionCall("Value (0-1)"=ComponentMask(None=MaterialFunctionCall() {materialFunction=&GetUserInterfaceUV}.Normalized UV) {g=true}) {materialFunction=&MF_UI_EaseCurves}
  n25 = ScalarParameter() {defaultValue=0.06}
  n26 = ScalarParameter() {defaultValue=0.4, parameterName=Param_1}
  n27 = ScalarParameter() {defaultValue=0.25, parameterName=Param_2}
  n28 = ScalarParameter() {defaultValue=0.01, parameterName=Param_3}
  NamedRerouteDeclaration(None=LinearInterpolate(
    A=VectorParameter() {parameterName=LinearGradientBottom, defaultValue=<rgba 0.7231 0.0976 0 1>}.RGB
    B=LinearInterpolate(A=@n22.RGB, B=@n23.RGB, Alpha=@n24.EaseOutCubic)
    Alpha=MaterialFunctionCall(GradientStart=@n25, GradientEnd=@n26, RotationAngle=@n27, GradientOffset=@n28) {materialFunction=&MF_UI_LinearGradient}.Result
   )) {name=Linear Gradient, nodeColor=<rgba 0.7231 0.0976 0 1>}
  n29 = MaterialFunctionCall() {materialFunction=&GetUserInterfaceUV}
  n30 = ScalarParameter() {defaultValue=1, group=BG Texture, parameterName=BG Scale, sortPriority=3}
  n31 = TextureObjectParameter() {
   group=BG Texture
   parameterName=Background Texture
   sortPriority=98
   texture=/BattlePassS30_Quests/Items/QuestRewards/FlatWare/Textures/T_UI_FlatWare_BG_Overlay
  }
  n32 = TextureSample(
   UVs=MaterialFunctionCall(UV=MaterialFunctionCall(UVs=@n29.9-Slice UV, Texture Scale=@n30) {materialFunction=&ScaleUVsByCenter}.UVs) {materialFunction=&MF_UI_UVScaling}.Scaled To Fill
   Tex=@n31
  )
  n33 = Multiply(
   A=Power(Base=@n32.RGB, Exp=ScalarParameter() {defaultValue=1.2, group=BG Texture, parameterName=Adjust Levels, sliderMax=3, sortPriority=4})
   B=ScalarParameter() {defaultValue=1, group=BG Texture, parameterName=Texture Opacity, sliderMax=1, sortPriority=2}
  )
  NamedRerouteDeclaration(None=Max(A=Saturate(None=@n33)) {constB=0.0001}) {name=BG Texture, nodeColor=<rgba 0 0 0 1>}
  n34 = NamedRerouteUsage()
  n35 = MaterialFunctionCall(
   UVs=MaterialFunctionCall() {materialFunction=&GetUserInterfaceUV}.Normalized UV
   Texture Scale=ScalarParameter() {defaultValue=1, group=Static Texture, parameterName=Texture Scale, sortPriority=3}
  ) {materialFunction=&ScaleUVsByCenter}
  n36 = TextureObjectParameter() {
   group=Static Texture
   parameterName=Static Texture
   samplerType=SAMPLERTYPE_Alpha
   sortPriority=98
   texture=/Game/UI/SurvivorTiers/QuestRewards/Textures/T_UI_SurvivorTierIcon_Nitro
  }
  n37 = Power(Base=TextureSample(UVs=MaterialFunctionCall(UV=@n35.UVs) {materialFunction=&MF_UI_UVScaling}.Scaled to Fit, Tex=@n36) {samplerSource=SSM_Clamp_WorldGroupSettings, samplerType=SAMPLERTYPE_Alpha}.R) {constExponent=0.6}
  NamedRerouteDeclaration(None=LinearInterpolate(A=Multiply(A=@n34) {constB=0.1}, Alpha=Power(Base=@n37)) {constA=1}) {name=Static Texture, nodeColor=<rgba 0 0 0 1>}
  NamedRerouteDeclaration(None=@n32.RGB) {name=BGTextureMask, nodeColor=<rgba 0 0 0 1>}
  n38 = LinearInterpolate(
   B=@n15
   Alpha=MaterialFunctionCall("Value (0-1)"=OneMinus(None=@n16)) {materialFunction=&MF_UI_EaseCurves}.EaseInOutSine
  ) {constA=1}
  n39 = OneMinus(None=ComponentMask(None=MaterialFunctionCall() {materialFunction=&GetUserInterfaceUV}.Normalized UV) {r=true})
  n40 = ScalarParameter() {defaultValue=1, group=Top Left Vignette, parameterName=Top Left Vignette Opacity}
  NamedRerouteDeclaration(None=LinearInterpolate(B=Saturate(None=@n38), Alpha=Multiply(A=@n39, B=@n40))) {name=Top Left Vignette, nodeColor=<rgba 0 0 0 1>}
  n41 = MaterialFunctionCall() {materialFunction=&GetUserInterfaceUV}
  n42 = Multiply(A=DotProduct(A=OneMinus(None=@n41.Normalized UV), B=@n41.Normalized UV)) {constB=2}
  n43 = ComponentMask(None=@n41.Normalized UV) {g=true}
  n44 = LinearInterpolate(
   B=Power(Base=Saturate(None=@n42)) {constExponent=0.7}
   Alpha=MaterialFunctionCall("Value (0-1)"=OneMinus(None=@n43)) {materialFunction=&MF_UI_EaseCurves}.EaseInQuint
  ) {constA=1}
  NamedRerouteDeclaration(None=LinearInterpolate(Alpha=Saturate(None=@n44)) {constB=0}) {name=Vignette_Bright, nodeColor=<rgba 0 0 0 1>}
  n45 = VectorParameter() {group=Left Bar, parameterName=Left Bar Color, sortPriority=36, defaultValue=<rgba 0.9301 1 0.0103 1>}
  n46 = ScalarParameter() {defaultValue=-0.25, group=Left Bar, parameterName=Left Bar Position, sortPriority=37}
  n47 = Multiply(A=Abs(None=@n46)) {constB=0.5}
  n48 = ScalarParameter() {group=Left Bar, parameterName=Left Bar Animation Transition, sliderMax=1, sortPriority=35}
  n49 = Multiply(A=@n48) {constB=2}
  n50 = LinearInterpolate(B=Multiply(A=Subtract(A=@n48) {constB=0.5}) {constB=2}, Alpha=Saturate(None=Floor(None=@n49))) {desc="AnimateHIghlight <= 0.5 result is 0"}
  n51 = Add(
   A=StaticSwitchParameter(True=LinearInterpolate(A=@n47, B=@n46, Alpha=@n50) {constA=-0.1, constAlpha=1}, False=@n46) {defaultValue=true, group=Left Bar, parameterName=Enable Left Bar Animation, sortPriority=34}
   B=MaterialFunctionCall() {materialFunction=&GetUserInterfaceUV}.9-Slice UV
  )
  n52 = ScalarParameter() {defaultValue=-0.25, group=Left Bar, parameterName=Left Bar Angle, sortPriority=38}
  n53 = Multiply(
   A=ComponentMask(None=Rotator(Coordinate=@n51, Time=@n52)) {r=true}
   B=ScalarParameter() {defaultValue=256, group=Left Bar, parameterName=Left Bar Sharpness, sortPriority=39}
  )
  n54 = NamedRerouteUsage()
  n55 = Multiply(A=NamedRerouteUsage(), B=Saturate(None=Add(B=@n54)))
  n56 = MaterialFunctionCall!(Base=@n55, Blend=NamedRerouteUsage()) {materialFunction=/Engine/Functions/Engine_MaterialFunctions03/Blends/Blend_SoftLight}
  n57 = MaterialFunctionCall!(
   Blend=NamedRerouteUsage()
   Base=StaticSwitchParameter(True=Saturate(None=@n56.Result), False=@n55) {defaultValue=true, group=BG Texture, parameterName=Enable BG Texture, sortPriority=1}
  ) {materialFunction=/Engine/Functions/Engine_MaterialFunctions03/Blends/Blend_ColorDodge}
  n58 = StaticSwitchParameter(True=Multiply(A=NamedRerouteUsage(), B=@n57.Result), False=@n57.Result) {group=Top Left Vignette, parameterName=Enable Top Left Dark Vignette}
  StaticSwitchParameter(True=LinearInterpolate(A=@n45.RGB, B=@n58, Alpha=Saturate(None=@n53)), False=@n58) {group=Left Bar, parameterName=Enable Left Bar, sortPriority=33}
  StaticSwitchParameter() {defaultValue=true, group=Static Texture, parameterName=Enable Static Texture, sortPriority=2}
 ]
}
