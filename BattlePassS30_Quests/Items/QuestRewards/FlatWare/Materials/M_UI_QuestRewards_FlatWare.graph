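# Translucent UI material for FlatWare quest reward images, drawing an icon with a tint, drop shadow and optional fade.
# Position and scale parameters transform the UVs, and the texture's own RGB blends from the tint color toward the image.
# A second texture sample, offset and smoothstepped on alpha, forms the shadow added to the icon's alpha.
# A static switch caps that alpha with a rotatable linear gradient for the Right Fade.
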Material /BattlePassS30_Quests/Items/QuestRewards/FlatWare/Materials/M_UI_QuestRewards_FlatWare
graph {
 properties {blendMode=BLEND_Translucent, materialDomain=MD_UI, pixelDepthOffsetMode=PDOM_Legacy, usedShadingModels=Default Lit}
 nodes [
  n1 = Multiply(A=ScalarParameter() {parameterName=Image Position X}) {constB=-1}
  n2 = ScalarParameter() {parameterName=Image Position Y}
  n3 = MaterialFunctionCall(
   UVs=Add(A=TextureCoordinate(), B=AppendVector(A=@n1, B=@n2))
   Texture Scale=AppendVector(A=ScalarParameter() {defaultValue=1, parameterName=Image Scale X}, B=ScalarParameter() {defaultValue=1, parameterName=Image Scale Y})
  ) {materialFunction=&ScaleUVsByCenter}
  n4 = TextureSampleParameter2D(UVs=@n3.UVs) {parameterName=Texture, texture=/Game/UI/Foundation/Textures/Icons/Weapons/Items/T-Icon-Pickaxes-Pickaxe-RavenQuill-L}
  LinearInterpolate(A=VectorParameter() {parameterName=Tint, defaultValue=<rgba 0 0.0009 0.0052 0>}.RGB, B=@n4.RGB, Alpha=@n4.RGB)
  n5 = ScalarParameter() {defaultValue=3, parameterName=U offset}
  n6 = ScalarParameter() {defaultValue=1, parameterName=V offset}
  n7 = Add(A=@n3.UVs, B=Multiply(A=AppendVector(A=@n5, B=@n6), B=Constant() {r=0.01}))
  n8 = SmoothStep(Value=TextureSampleParameter2D(UVs=Saturate(None=@n7)) {parameterName=Texture Drop Shadow, texture=&T-Icon-Pickaxes-Pickaxe-RavenQuill-L}.A) {constMin=0.8}
  n9 = ScalarParameter() {defaultValue=0.264, parameterName=Drop Shadow Opacity}
  n10 = ScalarParameter() {defaultValue=1, group=Right Fade, parameterName=Fade Start, sortPriority=34}
  n11 = ScalarParameter() {group=Right Fade, parameterName=Fade End, sortPriority=35}
  n12 = ScalarParameter() {defaultValue=1, group=Right Fade, parameterName=Fade Rotation, sliderMax=1, sortPriority=36}
  Min(
   A=Add(A=@n4.A, B=Multiply(A=@n8, B=@n9))
   B=StaticSwitchParameter(
    True=MaterialFunctionCall(GradientStart=@n10, GradientEnd=@n11, RotationAngle=@n12) {materialFunction=&MF_UI_LinearGradient}.Result
    False=Constant3Vector() {constant=<rgba 1 1 1 0>}
   ) {group=Right Fade, parameterName=Fade, sortPriority=33}
  )
 ]
}
